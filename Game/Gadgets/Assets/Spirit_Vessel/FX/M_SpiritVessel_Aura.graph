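# Translucent, unlit Niagara particle material for the Spirit Vessel gadget's aura, a portal-like glow with an opening driven by dynamic parameters.
# Radially panned caustics distort magnetic-wave lines, forming a tinted HDR edge glow.
# An eroded noise-distorted mask, sphere mask and depth fade build alpha, which also blends refraction.
# Side branches compute rotating normal-map vertex offset and a depth-offset, screen-space chromatic interior, the latter separate from the emissive chain.
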Material /Game/Gadgets/Assets/Spirit_Vessel/FX/M_SpiritVessel_Aura
graph {
 properties {
  bCastRayTracedShadows=false
  bUsedWithNiagaraMeshParticles=true
  bUsedWithNiagaraSprites=true
  blendMode=BLEND_Translucent
  pixelDepthOffsetMode=PDOM_Legacy
  refractionMethod=RM_IndexOfRefraction
  shadingModel=MSM_Unlit
  twoSided=true
  usedShadingModels=Unlit
 }
 nodes [
  n1 = TextureCoordinate()
  n2 = Normalize(VectorInput=AppendVector(A=Subtract(A=@n1) {constB=0.5}, B=Constant()))
  n3 = Multiply(A=NamedRerouteUsage())
  n4 = Multiply(A=ScalarParameter() {defaultValue=0.06, parameterName=ChromaticAmount}, B=Add(A=Add(), B=Add(B=@n3) {constA=0.2, constB=0.2}) {constB=0.2})
  n5 = MaterialFunctionCall(Float3=Multiply(B=Multiply(A=@n2, B=@n4) {constB=2})) {desc=Chromatic offset, materialFunction=&BreakOutFloat3Components}
  n6 = ComponentMask() {g=true, r=true}
  n7 = ScalarParameter() {defaultValue=0.4, parameterName=EdgeRippleDistortion}
  n8 = TransformPosition(None=MaterialFunctionCall(X=ScalarParameter() {defaultValue=-1500, parameterName=DepthOffset}) {materialFunction=&MakeFloat3}.Result) {transformSourceType=TRANSFORMPOSSOURCE_Particle, transformType=TRANSFORMPOSSOURCE_World}
  n9 = Add(A=Subtract(
    A=ScreenPosition().ViewportUV
    B=MaterialFunctionCall(World Position=@n8) {materialFunction=&TransformToClipSpace}.Buffer UVs for Scene Textures
   )) {constB=0.5}
  n10 = ScalarParameter() {defaultValue=1200, parameterName=DistanceScale}
  n11 = Distance(A=ObjectPositionWS(), B=CameraPositionWS())
  n12 = Constant() {r=1}
  n13 = MaterialFunctionCall(Float2=ViewSize()) {materialFunction=&BreakOutFloat2Components}
  n14 = Divide(A=@n13.R, B=@n13.R)
  n15 = ScalarParameter() {defaultValue=1, parameterName=ScaleMult}
  n16 = LinearInterpolate(Alpha=NamedRerouteUsage()) {constA=0.4}
  n17 = Multiply(
   A=Clamp(None=Divide(A=@n10, B=@n11), Max=ScalarParameter() {defaultValue=0.75, parameterName=MaxDistanceScaleUp}) {clampMode=CMODE_ClampMax, maxDefault=0}
   B=Multiply(A=AppendVector(A=@n12, B=@n14), B=Multiply(A=@n15, B=@n16))
  )
  n18 = Add(
   A=NamedRerouteDeclaration(None=MaterialFunctionCall(UVs=@n9, Texture Scale=@n17) {materialFunction=&ScaleUVsByCenter}.UVs) {name=PortalUVs, nodeColor=<rgba 0 1 0.7188 1>}
   B=Multiply(B=Multiply(A=@n6, B=@n7))
  ) {constB=0}
  n19 = TextureSampleParameter2D(UVs=@n18) {parameterName=InteriorTexture, texture=/Game/Effects/Fort_Effects/Textures/Tile/T_Caustics_Chromatic}
  n20 = MaterialFunctionCall(
   R=TextureSampleParameter2D(UVs=Add(A=@n18, B=@n5.R)) {parameterName=InteriorTexture, texture=&T_Caustics_Chromatic}.R
   G=TextureSampleParameter2D(UVs=Add(A=@n18, B=@n5.G)) {parameterName=InteriorTexture, texture=&T_Caustics_Chromatic}.G
   B=@n19.B
  ) {materialFunction=&AppendMany}
  Add(B=QualitySwitch(Default=ShadingPathSwitch(Default=@n20.RGB, Forward=@n19.RGB, Mobile=@n19.RGB), Low=@n19.RGB, Medium=@n20.RGB))
  Constant() {r=3.5}
  n21 = ScalarParameter() {defaultValue=5, parameterName=DepthFadeDist}
  n22 = DynamicParameter() {parameterIndex=1, defaultValue=<rgba 0 0 0.8 0>, paramNames [DistortionPanSpeed, MaskDistortPanSpeed, MaskUVScale, MaskErosionStrength]}
  n23 = TextureSampleParameter2D(UVs=Panner(Coordinate=TextureCoordinate() {coordinateIndex=1}, Speed=@n22.MaskDistortPanSpeed)) {parameterName=DistortionNoise, samplerSource=SSM_Wrap_WorldGroupSettings, texture=&T_NoisePack_01}
  n24 = ComponentMask(None=Subtract(A=Multiply(A=@n23.RGB) {constB=2})) {g=true, r=true}
  n25 = DynamicParameter() {parameterIndex=2, defaultValue=<rgba 0.1 2 1 1>, paramNames [MaskDistortionStrength, DarkStrength, none2, none3]}
  n26 = TextureCoordinate() {coordinateIndex=1}
  n27 = TextureSampleParameter2D(UVs=Add(
    A=Multiply(A=@n24, B=@n25.MaskDistortionStrength)
    B=MaterialFunctionCall(UVs=@n26, Texture Scale=@n22.MaskUVScale) {materialFunction=&ScaleUVsByCenter}.UVs
   )) {parameterName=Mask Tex, samplerSource=SSM_Clamp_WorldGroupSettings, texture=/Game/Effects/Fort_Effects/Textures/Mask/T_Sphere_A}
  n28 = DynamicParameter() {parameterIndex=1, defaultValue=<rgba 0 0 0.8 0>, paramNames [DistortionPanSpeed, MaskDistortPanSpeed, MaskUVScale, MaskErosionStrength]}
  n29 = Power(Base=Subtract(A=@n27.RGB, B=@n28.MaskErosionStrength))
  n30 = TextureCoordinate() {coordinateIndex=1}
  n31 = Constant() {r=0.5}
  n32 = DynamicParameter() {defaultValue=<rgba 1 1 0 1>, paramNames [Open, Warp, WPOAmount, ScaleUVDistortion]}
  n33 = LinearInterpolate(Alpha=NamedRerouteDeclaration(None=@n32.Open) {name=OpenParam, nodeColor=<rgba 1 0.8906 0 1>}) {constB=0.4}
  n34 = Multiply(A=Power(Base=SphereMask(A=@n30, B=@n31, Radius=@n31, Hardness=@n33) {hardnessPercent=0.7})) {constB=1.4}
  n35 = DynamicParameter() {parameterIndex=2, defaultValue=<rgba 0.1 2 1 1>, paramNames [MaskDistortionStrength, DarkStrength, none2, none3]}
  n36 = TextureCoordinate() {coordinateIndex=1}
  n37 = NamedRerouteUsage()
  n38 = MaterialFunctionCall(Vector or UVs=Add(A=MaterialFunctionCall(UVs=@n36, Texture Scale=@n37) {materialFunction=&ScaleUVsByCenter}.UVs) {constB=-0.5}) {materialFunction=&VectorToRadialValue}
  n39 = VectorParameter() {parameterName=DistortionTile, defaultValue=<rgba 1 2.5 2 0>}
  n40 = AppendVector(A=@n39.R, B=@n39.R)
  n41 = DynamicParameter() {parameterIndex=1, defaultValue=<rgba 0 0 0.8 0>, paramNames [DistortionPanSpeed, MaskDistortPanSpeed, MaskUVScale, MaskErosionStrength]}
  n42 = Multiply(A=@n41.DistortionPanSpeed, B=Time())
  n43 = Multiply(A=NamedRerouteUsage()) {constB=3}
  n44 = TextureCoordinate() {coordinateIndex=1}
  n45 = TextureSampleParameter2D!(
   UVs=Panner(Coordinate=Multiply(A=@n38.Radial Coordinates, B=@n40), Time=Add(A=@n42, B=@n43)) {speedX=0.1, speedY=-0.22}
   "DDX(UVs)"=DDX(Value=@n44)
   "DDY(UVs)"=DDY(Value=@n44)
   Apply View MipBias
  ) {
   mipValueMode=TMVM_Derivative
   parameterName=DistortionTexture
   texture=/Game/Effects/Fort_Effects/Textures/Tile/T_ShieldCaustics_Tile_Chromatic
  }
  n46 = Subtract(A=ComponentMask(None=@n45.RGB) {g=true, r=true}) {constB=0.5}
  n47 = TextureCoordinate() {coordinateIndex=1}
  n48 = SphereMask(A=@n47, B=Constant() {r=0.5}, Radius=ScalarParameter() {defaultValue=0.8, parameterName=InteriorLinesRadius}) {attenuationRadius=500, hardnessPercent=0}
  n49 = Power(Base=@n48, Exp=ScalarParameter() {defaultValue=5, parameterName=InteriorLinesPower})
  n50 = AppendVector(A=@n49, B=ComponentMask(None=@n47) {g=true})
  n51 = VectorParameter() {parameterName=LinesTile, defaultValue=<rgba 4 0.7 0 0>}
  n52 = NamedRerouteUsage()
  n53 = Multiply(A=AppendVector(A=@n51.R, B=@n51.R), B=LinearInterpolate(Alpha=@n52) {constA=0.25})
  n54 = Time()
  n55 = ScalarParameter() {defaultValue=1, parameterName=InteriorLineSPanSpeed}
  n56 = Panner(Coordinate=Add(A=Multiply(A=@n46) {constB=0.1}, B=Multiply(A=@n50, B=@n53) {constB=2}), Time=Add(A=Multiply(A=@n54, B=@n55), B=@n52)) {speedX=0.24, speedY=-0.14}
  n57 = LinearInterpolate(
   A=TextureSample(UVs=@n56) {texture=/Game/Athena/Environments/Nexus/Textures/T_ZeroPoint_MagneticWaves_Blurred}.RGB
   B=TextureSample(UVs=@n56) {texture=/Game/Athena/Environments/Nexus/Textures/T_ZeroPoint_MagneticWaves}.RGB
   Alpha=@n49
  )
  n58 = LinearInterpolate(A=@n35.DarkStrength, Alpha=ComponentMask(None=@n57) {g=true}) {constA=2.5, constB=0.5}
  n59 = Multiply(A=OneMinus(None=ComponentMask(None=@n29) {r=true}), B=Saturate(None=Multiply(A=@n34, B=@n58)))
  n60 = ParticleColor()
  n61 = Multiply(A=ShadingPathSwitch(Default=@n59, Deferred=DepthFade(Opacity=@n59, FadeDistance=@n21)), B=@n60.A)
  LinearInterpolate(B=ScalarParameter() {defaultValue=1.25, parameterName=Refraction}, Alpha=@n61) {constA=1}
  LinearInterpolate() {constA=0.7, constB=2}
  n62 = VectorParameter() {parameterName=EdgeColor, defaultValue=<rgba 1.8968 2.0261 4 0>}
  n63 = LinearInterpolate(Alpha=@n48) {constA=1, constB=-0.45}
  n64 = Desaturation(None=Multiply(A=Saturate(None=@n63), B=@n57) {constB=0}, Fraction=Constant() {r=-1})
  n65 = Add(A=@n45.RGB, B=@n64)
  n66 = TextureCoordinate() {coordinateIndex=1}
  n67 = Constant() {r=0.5}
  n68 = LinearInterpolate(Alpha=NamedRerouteUsage()) {constA=0.1, constB=0.5}
  n69 = ScalarParameter() {defaultValue=1.5, parameterName=EdgeMaskPower}
  n70 = NamedRerouteUsage()
  n71 = LinearInterpolate(A=@n45.RGB, B=@n64) {constAlpha=0.6}
  n72 = Power(
   Base=OneMinus(None=SphereMask(A=@n66, B=@n67, Radius=@n68) {hardnessPercent=0})
   Exp=LinearInterpolate(
    A=LinearInterpolate(A=@n69, Alpha=@n70) {constB=0.6}
    B=ScalarParameter() {defaultValue=0.25, parameterName=EdgeMaskPowerMin}
    Alpha=Desaturation(None=@n71)
   ) {constA=3, constB=0.25}
  ) {constExponent=0.5}
  n73 = Multiply(A=Multiply(A=Multiply(A=@n62.RGB, B=@n65), B=Multiply(A=@n72, B=@n72)), B=@n60.RGB)
  n74 = Clamp(None=@n73)
  n75 = ScalarParameter() {defaultValue=1.25, parameterName=HDRIntensity}
  MaterialFunctionCall(Emissive=ShadingPathSwitch(
    Default=@n73
    Deferred=MaterialFunctionCall(LDR Input=@n74, HDR Tint=@n73, HDR Intensity Multiplier=@n75) {materialFunction=&DeriveHDRfromLDR}.RemappedRange
   )) {materialFunction=&MF_FX_EmissiveAdjustments}
  MaterialFunctionCall(Alpha Threshold=@n61) {materialFunction=&DitherTemporalAA}
  MaterialFunctionCall(Radius=LinearInterpolate(B=Constant() {r=0.7})) {materialFunction=&RadialGradientExponential}
  n76 = Multiply(A=Time(), B=ScalarParameter() {defaultValue=1, parameterName=WPOSpeed})
  n77 = Rotator(
   Coordinate=MaterialFunctionCall(
    UVs=TextureCoordinate() {coordinateIndex=1, uTiling=0.5, vTiling=0.5}
    Texture Scale=ScalarParameter() {defaultValue=1, parameterName=ScaleUVsWPO}
   ) {materialFunction=&ScaleUVsByCenter}.UVs
   Time=@n76
  ) {centerX=0.25, centerY=0.25, speed=0.05}
  n78 = ComponentMask(None=TextureSample(UVs=Panner(Coordinate=@n77, Time=@n76) {speedX=0.05, speedY=0.1}) {samplerType=SAMPLERTYPE_Normal, texture=&T_Weap_Distort_N_02}.RGB) {g=true, r=true}
  n79 = Constant()
  Multiply(A=Transform(None=AppendVector(A=@n78, B=@n79)) {transformSourceType=TRANSFORMSOURCE_Local}, B=NamedRerouteUsage())
  NamedRerouteDeclaration(None=@n32.Warp) {name=WarpParam, nodeColor=<rgba 1 0.4038 0 1>}
  NamedRerouteDeclaration(None=@n32.WPOAmount) {name=WPOAmount, nodeColor=<rgba 1 0 0.8203 1>}
  NamedRerouteDeclaration(None=@n32.ScaleUVDistortion) {name=ScaleUVDistortion, nodeColor=<rgba 0 0.1562 1 1>}
  ScalarParameter() {defaultValue=4, parameterName=ScaleUVsDistortion}
  ScalarParameter() {defaultValue=0.7, parameterName=PanSpeed}
  ScalarParameter() {defaultValue=200, parameterName=WPO}
  TextureCoordinate() {coordinateIndex=1}
 ]
}
