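# Unlit, two-sided translucent material for the Spirit Vessel gadget's tendril effects, built for Niagara mesh particles and ribbons.
# Color blends curve-atlas gradients, a sphere-masked emissive around the object and a panning spooky-faces texture, tinted by particle color.
# Opacity combines that texture with particle alpha, Fresnel and depth fade.
# World position offset adds per-particle randomized sine or cosine wobble to SplineThicken, with width and motion driven by dynamic parameters.
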Material /Game/Gadgets/Assets/Spirit_Vessel/FX/M_SpiritVessel_Tendrils
graph {
 properties {
  bUsedWithNiagaraMeshParticles=true
  bUsedWithNiagaraRibbons=true
  blendMode=BLEND_Translucent
  pixelDepthOffsetMode=PDOM_Legacy
  shadingModel=MSM_Unlit
  twoSided=true
  usedShadingModels=Unlit
 }
 nodes [
  n1 = CurveAtlasRowParameter(CurveTime=MaterialFunctionCall(Float2=NamedRerouteUsage()) {materialFunction=&BreakOutFloat2Components}.R) {
   atlas=/Game/Effects/Fort_Effects/FrontEnd/SpyBase/Materials/CurveAtlas_2
   curve=/Game/Effects/Fort_Effects/FrontEnd/SpyBase/Materials/Curve_FX_Mask2_2
   defaultValue=1
   parameterName=SpiritCurve
  }
  n2 = NamedRerouteUsage()
  n3 = LinearInterpolate(
   A=VectorParameter() {parameterName=ShadowColor, defaultValue=<rgba 0.0365 0.0056 0.0016 1>}.RGB
   B=VectorParameter() {parameterName=LightColor, defaultValue=<rgba 0.651 0.069 0.0106 1>}.RGB
   Alpha=Multiply(A=@n1.A, B=LinearInterpolate(Alpha=@n2))
  )
  n4 = VectorParameter() {parameterName=CenterEmissve, defaultValue=<rgba 0.6386 0.0793 2 1>}
  n5 = LinearInterpolate(A=Divide(A=@n4.RGB) {constB=10}, B=@n4.RGB, Alpha=@n1.R)
  n6 = CurveAtlasRowParameter(CurveTime=SphereMask(
    A=WorldPosition().XYZ
    B=ObjectPositionWS()
    Radius=ScalarParameter() {defaultValue=240, parameterName=WorldSphereEmissiveRadius}
    Hardness=Constant()
   )) {
   atlas=&CurveAtlas_2
   curve=&Curve_FX_Mask2_2
   defaultValue=1
   parameterName=SpiritCurve
  }
  n7 = Saturate(None=@n6.B)
  n8 = NamedRerouteUsage()
  n9 = DynamicParameter() {parameterIndex=1, defaultValue=<rgba 0.1 1 1 1>, paramNames [MaskPanSpeedX, MaskPanSpeedY, MaskTexX, MaskTexY]}
  n10 = AppendVector(A=@n9.MaskTexX, B=@n9.MaskTexX)
  n11 = TextureSampleParameter2D(UVs=Panner(
    Coordinate=MaterialFunctionCall(UVs=@n8, Texture Scale=@n10) {materialFunction=&ScaleUVsByCenter}.UVs
    Speed=AppendVector(A=@n9.MaskPanSpeedX, B=@n9.MaskPanSpeedX)
   ) {speedX=0.1, speedY=1}) {parameterName=MaskTex, texture=/Game/Effects/Fort_Effects/Textures/Tile/T_SpookyFaces_Chroma_Tile}
  n12 = VectorParameter() {parameterName=SpecColor, defaultValue=<rgba 0.0484 0.083 0.6406 1>}
  n13 = ParticleColor()
  n14 = Multiply(A=Add(A=LinearInterpolate(A=@n3, B=@n5, Alpha=@n7), B=Multiply(A=@n11.R, B=@n12.RGB)) {constB=0}, B=@n13.RGB)
  MaterialFunctionCall(Emissive=@n14) {materialFunction=&MF_FX_EmissiveAdjustments}
  NamedRerouteDeclaration(None=@n14) {name=Color, nodeColor=<rgba 0.8234 0.0183 1 1>}
  n15 = Multiply(A=@n13.A, B=@n11.R)
  n16 = CurveAtlasRowParameter(CurveTime=MaterialFunctionCall(Float2=NamedRerouteUsage()) {materialFunction=&BreakOutFloat2Components}.G) {
   atlas=&CurveAtlas_2
   curve=/Game/Effects/Fort_Effects/FrontEnd/SpyBase/Materials/Curve_Color_EnergyCore_Blue_2
   defaultValue=3
   parameterName=SpiritCurve
  }
  n17 = Multiply(
   A=@n16.R
   B=Multiply(A=ScalarParameter() {defaultValue=40, parameterName=SplineWidth}, B=DynamicParameter() {paramNames [X,Y,Z,SplineWidth]}.SplineWidth)
  )
  n18 = MaterialFunctionCall(WidthBase=@n17, WidthTip=@n17) {materialFunction=&SplineThicken}
  n19 = MaterialFunctionCall(
   Normal Vector=@n18.Normal
   Invert Fresnel=StaticBoolParameter() {defaultValue=true, parameterName=InvertFresnel}
   Power=ScalarParameter() {defaultValue=1, parameterName=FresnelExponent}
  ) {materialFunction=&Fresnel_Function}
  n20 = ScalarParameter() {defaultValue=10, parameterName=DepthFadeDist}
  NamedRerouteDeclaration(None=Multiply(A=Multiply(A=@n15, B=@n19.Result), B=DepthFade(FadeDistance=@n20))) {name=Opacity, nodeColor=<rgba 0 1 0.5078 1>}
  n21 = DynamicParameter() {paramNames [X,Y,Z,SplineWidth]}
  n22 = NamedRerouteUsage()
  n23 = ScalarParameter() {defaultValue=1.8, parameterName=WPO_Tile}
  n24 = Divide(A=Time()) {constB=6}
  n25 = Multiply(A=ParticleRandom()) {constB=3}
  n26 = MaterialFunctionCall(Float2=Add(A=Multiply(A=@n22, B=@n23), B=Add(A=@n24, B=@n25))) {materialFunction=&BreakOutFloat2Components}
  n27 = ParticleRandom()
  n28 = Multiply(
   A=MaterialFunctionCall(R=@n21.X, G=@n21.X, B=@n21.X) {materialFunction=&AppendMany}.RGB
   B=LinearInterpolate(A=Sine(None=@n26.G), B=Cosine(None=@n26.G), Alpha=Round(None=@n27))
  )
  n29 = ScalarParameter() {defaultValue=40, parameterName=WPO_Intensity}
  n30 = ScalarParameter() {defaultValue=1, parameterName=UseSplineMesh}
  NamedRerouteDeclaration(None=Add(A=Multiply(A=@n28, B=@n29), B=LinearInterpolate(B=@n18.WorldPositionOffset, Alpha=@n30))) {name=WPO, nodeColor=<rgba 0 0.7656 1 1>}
  n31 = TextureCoordinate()
  NamedRerouteDeclaration(None=LinearInterpolate(
    A=@n31
    B=MaterialFunctionCall(XY=@n31) {materialFunction=&Swizzle}.YX
    Alpha=ScalarParameter() {parameterName=SwizzleUV}
   )) {name=TextureCoord, nodeColor=<rgba 1 0 0.0343 1>}
  n32 = NamedRerouteUsage()
  n33 = ComponentMask(None=VectorParameter() {parameterName=MaskNoiseTile, defaultValue=<rgba 1 2 0 0>}.RGB) {g=true, r=true}
  n34 = Multiply(A=@n33) {constB=0.7}
  n35 = TextureSampleParameter2D(UVs=Panner(Coordinate=Multiply(A=@n32, B=@n34)) {speedX=0.2, speedY=0.3}) {parameterName=NoiseA, samplerType=SAMPLERTYPE_LinearGrayscale, texture=/Game/Environments/Hexmap/HexClouds/T_HexmapFog_01_Height02}
  n36 = Multiply(A=@n33) {constB=0.5}
  n37 = NamedRerouteUsage()
  n38 = TextureSampleParameter2D(UVs=Panner(Coordinate=Multiply(A=@n36, B=@n37)) {speedX=0.1, speedY=0.7}) {
   parameterName=NoiseB
   samplerType=SAMPLERTYPE_LinearGrayscale
   texture=/Game/Global/Textures/Default/Tileable_Patterns/T_StormCloudDetail01_M
  }
  NamedRerouteDeclaration(None=Multiply(A=@n35.R, B=@n38.R)) {name=NoiseInput, nodeColor=<rgba 1 0 0.3984 1>}
  NamedRerouteUsage()
  NamedRerouteUsage()
  NamedRerouteUsage()
  n39 = CameraVectorWS()
  n40 = VertexNormalWS()
  n41 = Saturate(None=Abs(None=DotProduct(A=@n39, B=@n40)))
  Power(Base=Multiply(A=@n41, B=@n41), Exp=Constant() {r=0.8})
 ]
}
